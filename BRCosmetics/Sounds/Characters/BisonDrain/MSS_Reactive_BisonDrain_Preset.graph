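# Stereo one-shot MetaSound source for the BisonDrain reactive cosmetic, playing a chain of laugh and chuckle VO clips through the shared MSS_Cosmetic_Simple_Base graph.
# When each clip finishes, a probability filter either retriggers the base after a short random delay or ends the sound.
# Deferred variable accessors close the feedback loop, and each repeat lowers the next repeat's probability, so the laughter chain tapers off.
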MetaSoundSource /BRCosmetics/Sounds/Characters/BisonDrain/MSS_Reactive_BisonDrain_Preset
inputs [
 PitchModulationRange.Max {type=Float}
 PitchModulationRange.Min {type=Float}
 StartDelay.Max {type=Time}
 StartDelay.Min {type=Time}
 TemplatePitchMultiplier {type=Float}
 TemplateVolumeMultiplier {type=Float}
 UE.Source.OnPlay {type=Trigger}
 VolumeModulationRange.Max {type=Float}
 VolumeModulationRange.Min {type=Float}
 Waves {type=WaveAsset:Array}
]
interfaces [UE.OutputFormat.Stereo:1.0, UE.Source.OneShot:1.0, UE.Source:1.1]
outputs [
 "UE.OutputFormat.Stereo.Audio:0" {type=Audio}
 "UE.OutputFormat.Stereo.Audio:1" {type=Audio}
 UE.Source.OneShot.OnFinished {type=Trigger}
]
pages [Default {
  input_defaults {
   PitchModulationRange.Max=1
   PitchModulationRange.Min=1
   StartDelay.Max=0
   StartDelay.Min=0
   TemplatePitchMultiplier=1
   TemplateVolumeMultiplier=1
   UE.Source.OnPlay=null
   VolumeModulationRange.Max=1
   VolumeModulationRange.Min=1
   Waves [
    &/VO_BisonDrain_LaughsChuckles_01
    &/VO_BisonDrain_LaughsChuckles_02
    &/VO_BisonDrain_LaughsChuckles_03
    &/VO_BisonDrain_LaughsChuckles_04
    &/VO_BisonDrain_LaughsChuckles_10
    &/VO_BisonDrain_LaughsChuckles_11
    &/VO_BisonDrain_LaughsChuckles_12
    &/VO_BisonDrain_LaughsChuckles_13
    &/VO_BisonDrain_LaughsChuckles_14
    &/VO_BisonDrain_LaughsChuckles_16
    &/VO_BisonDrain_LaughsChuckles_17
   ]
  }
  nodes [
   InitVariable.Trigger()
   n1 = VariableDeferredAccessor.Float(Variable=InitVariable.Float().Variable)
   n2 = VariableDeferredAccessor.Float(Variable=@n1.Variable)
   n3 = VariableDeferredAccessor.Trigger(Variable=InitVariable.Trigger().Variable)
   TriggerFilter_0 = UE.Trigger Filter(Probability=@n1.Value, Trigger=@n3.Value)
   n4 = Value.Value:Float(
    Target Value=UE.Subtract:Float(PrimaryOperand=@n2.Value) {also {defaults {AdditionalOperands=0.25}}}.Out
    Set=@TriggerFilter_0.Heads
    Init Value=@n2.Value
   )
   VariableMutator.Float(Value=@n4.Output Value, Variable=@n1.Variable)
   n5 = InitVariable.Trigger()
   TriggerAnyTriggerAny2_0 = "TriggerAny.Trigger Any (2)"(In 0=^UE.Source.OnPlay, In 1=@n4.On Set)
   n6 = UE.RandomTime(Next=@TriggerAnyTriggerAny2_0.Out) {also {defaults {Max=0.15}}}
   TriggerDelay_0 = UE.Trigger Delay(In=@n6.On Next, Delay Time=@n6.Value)
   MSS_Cosmetic_Simple_Base_0 = /FortMetasound/Cosmetics/MSS_Cosmetic_Simple_Base(
    PitchModulationRange.Max=^PitchModulationRange.Max
    PitchModulationRange.Min=^PitchModulationRange.Min
    StartDelay.Max=^StartDelay.Max
    StartDelay.Min=^StartDelay.Min
    TemplatePitchMultiplier=^TemplatePitchMultiplier
    TemplateVolumeMultiplier=^TemplateVolumeMultiplier
    VolumeModulationRange.Max=^VolumeModulationRange.Max
    VolumeModulationRange.Min=^VolumeModulationRange.Min
    Waves=^Waves
    UE.Source.OnPlay=VariableAccessor.Trigger(Variable=VariableMutator.Trigger(Variable=@n5.Variable, Value=@TriggerDelay_0.Out).Variable).Value
   )
   VariableMutator.Trigger(Value=@MSS_Cosmetic_Simple_Base_0.UE.Source.OneShot.OnFinished, Variable=@n3.Variable)
  ]
  links [
   {dst="@UE.OutputFormat.Stereo.Audio:0", src=MSS_Cosmetic_Simple_Base_0.UE.OutputFormat.Stereo.Audio:0}
   {dst="@UE.OutputFormat.Stereo.Audio:1", src=MSS_Cosmetic_Simple_Base_0.UE.OutputFormat.Stereo.Audio:1}
   {dst="@UE.Source.OneShot.OnFinished", src=TriggerFilter_0.Tails}
  ]
 }]
properties {
 assetImportData=&self:AssetImportData
 attenuationSettings=&Fort_Pets_Close_Athena_Att
 audioPropertiesSheet=/BRCosmetics/Sounds/Settings/PropertySheets/Reactive/APS_Cosmetic_Reactive_Skin
 maxDistance=1800.0001
 numChannels=2
 outputFormat=Stereo
 sampleRate=48000
 soundClassObject=/Game/Sounds/FortSoundClassesAndModes/Fort_Backblings
 volume=1.5
 concurrencySet [&/CS_BisonDrain, &/CS_BisonDrain_Global]
}
